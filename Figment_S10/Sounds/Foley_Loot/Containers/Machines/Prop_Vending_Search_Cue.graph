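# Sound cue for searching a vending machine container, built from borrowed fridge and gear foley.
# A mixer layers two random picks: one of three fridge search waves at low volume and one of three gear search waves at boosted volume.
# It uses prop search attenuation and the searching foley sound class, with a concurrency override that prevents new instances once the limit is reached.
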SoundCue /Figment_S10/Sounds/Foley_Loot/Containers/Machines/Prop_Vending_Search_Cue
root=Mixer_8
nodes [
 {id=Mixer_8, children [Random_2,Random_7], properties {inputVolume [0.25,1.2]}}
 {id=Random_2, children [WavePlayer_10, WavePlayer_11, WavePlayer_12], properties {weights [1,1,1]}}
 {id=Random_7, children [WavePlayer_33, WavePlayer_34, WavePlayer_36], properties {weights [1,1,1]}}
 {id=WavePlayer_10, properties {soundWaveAssetPtr=/Game/Sounds/Foley_Loot/Containers/Kitchen/Prop_Fridge_Search_01}}
 {id=WavePlayer_11, properties {soundWaveAssetPtr=/Game/Sounds/Foley_Loot/Containers/Kitchen/Prop_Fridge_Search_02}}
 {id=WavePlayer_12, properties {soundWaveAssetPtr=/Game/Sounds/Foley_Loot/Containers/Kitchen/Prop_Fridge_Search_03}}
 {id=WavePlayer_33, properties {soundWaveAssetPtr=/Game/Sounds/Foley_Loot/Containers/Misc/Prop_Gear_Search_05}}
 {id=WavePlayer_34, properties {soundWaveAssetPtr=/Game/Sounds/Foley_Loot/Containers/Misc/Prop_Gear_Search_06}}
 {id=WavePlayer_36, properties {soundWaveAssetPtr=/Game/Sounds/Foley_Loot/Containers/Misc/Prop_Gear_Search_01}}
]
properties {
 attenuationSettings=/Game/Sounds/Attenuation/Fort_Prop_Search_Att
 bOverrideConcurrency=true
 maxDistance=1300
 soundClassObject=/Figment_S10/Sounds/FortSoundClassesAndModes/Fort_Searching_Foley
 volumeMultiplier=1.3
 concurrencyOverrides {resolutionRule=PreventNew}
}
